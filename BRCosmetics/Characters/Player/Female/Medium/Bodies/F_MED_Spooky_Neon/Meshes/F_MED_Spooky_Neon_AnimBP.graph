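# Post-process style cosmetic AnimBP for the F_MED_Spooky_Neon body, copying the attached parent's pose and adding corrective and secondary motion.
# Each update maps thigh yaw relative to the pelvis into per-side alphas, scaled down when PawnSpeed2D exceeds 100.
# Those alphas drive additive glute bone transforms, followed by two RigidBody nodes simulating bandana and jacket physics assets in spine_02 space.
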AnimBlueprint /BRCosmetics/Characters/Player/Female/Medium/Bodies/F_MED_Spooky_Neon/Meshes/F_MED_Spooky_Neon_AnimBP
target_skeleton=&/F_MED_Spooky_Neon_Skeleton
graphs [
 AnimGraph {nodes [
   n1 = "Transform (Modify) Bone - Bone: dyn_glute_l"(
    ComponentPose=Local To Component(LocalPose=Copy Pose From Mesh() {bUseAttachedParent=true})
    Alpha=Get thigh_rot_l()
    Translation="-3.750,0.875,-0.9"
    Rotation="0.000,0.000,-55"
    Scale="1.25,1.25,1.25"
   ) {
    rotationMode=BMM_Additive
    rotationSpace=BCS_BoneSpace
    scaleMode=BMM_Additive
    scaleSpace=BCS_BoneSpace
    translationMode=BMM_Additive
    translationSpace=BCS_BoneSpace
    alphaBoolBlend {blendOption=Linear}
    boneToModify {boneName=dyn_glute_l}
    rotation {roll=-55}
    scale=<1.25 1.25 1.25>
    translation=<-3.75 0.875 -0.9>
   }
   n2 = "Transform (Modify) Bone - Bone: dyn_glute_r"(ComponentPose=@n1, Alpha=Get thigh_rot_r(), Translation="-3.75,0,0", Rotation="0.000,0.000,-55", Scale="1.3,1.3,1.3") {
    rotationMode=BMM_Additive
    rotationSpace=BCS_BoneSpace
    scaleMode=BMM_Additive
    scaleSpace=BCS_BoneSpace
    translationMode=BMM_Additive
    translationSpace=BCS_BoneSpace
    alphaBoolBlend {blendOption=Linear}
    boneToModify {boneName=dyn_glute_r}
    rotation {roll=-55}
    scale=<1.3 1.3 1.3>
    translation {x=-3.75}
   }
   n3 = RigidBody(ComponentPose=@n2) {
    bClampLinearTranslationLimitToRefPose=true
    bOverrideWorldGravity=true
    overridePhysicsAsset=&/F_MED_Spooky_Neon_Bandana_Physics
    simulationSpace=BaseBoneSpace
    alphaBoolBlend {blendOption=Linear}
    baseBoneRef {boneName=spine_02}
    componentAppliedLinearAccClamp=<3000 3000 3000>
    componentLinearAccScale=<0.5 0.2 0.125>
    componentLinearVelScale=<0.5 0.2 0.125>
    overrideWorldGravity {z=-980}
   }
   n4 = RigidBody(ComponentPose=@n3) {
    bClampLinearTranslationLimitToRefPose=true
    bOverrideWorldGravity=true
    overridePhysicsAsset=&/F_MED_Spooky_Neon_Jacket
    simulationSpace=BaseBoneSpace
    alphaBoolBlend {blendOption=Linear}
    baseBoneRef {boneName=spine_02}
    componentAppliedLinearAccClamp=<3000 3000 2800>
    componentLinearAccScale=<0.4 0.4 0.2>
    componentLinearVelScale=<0.4 0.4 0.2>
    overrideWorldGravity {z=-980}
   }
   Output Pose(Result=Component To Local(ComponentPose=@n4)) {layerGroup=None}
  ]}
 EventGraph {nodes [
   Event BlueprintUpdateAnimation()
   > Get!(execute:Exec) -> (then:Exec, else:Exec, Skeletal_Mesh:Skeletal Mesh Component Object Reference)
   > Sequence()
   > n1 = Branch(Condition="float > float"(A=Get PawnSpeed2D(), B=100) {is=title K2Node_CallFunction})
   n2 = Get Skeletal_Mesh() -> (Skeletal_Mesh:Skeletal Mesh Component Object Reference)
   n3 = GetDeltaTransformFromRefPose(self=@n2, BoneName=thigh_l, BaseName=pelvis)
   n4 = MapRangeClamped(Value=BreakRotator(InRot=BreakTransform(InTransform=@n3).Rotation).Yaw, InRangeA=-20, InRangeB=-100, OutRangeB=0.7)
   > Set thigh_rot_l(thigh_rot_l=float * float(A=@n4, B=0.25))
   n5 = GetDeltaTransformFromRefPose(self=@n2, BoneName=thigh_r, BaseName=pelvis)
   n6 = MapRangeClamped(Value=BreakRotator(InRot=BreakTransform(InTransform=@n5).Rotation).Yaw, InRangeA=-20, InRangeB=-100)
   > Set thigh_rot_r(thigh_rot_r=float * float(A=@n6, B=0.5))
   Set thigh_rot_l(execute=@n1.else, thigh_rot_l=@n4)
   > Set thigh_rot_r(thigh_rot_r=@n6)
   Event BlueprintInitializeAnimation()
   > Set Skeletal_Mesh(Skeletal_Mesh:Skeletal Mesh Component Object Reference=GetOwnerSkeletalMeshComponent()) -> (Output_Get:Skeletal Mesh Component Object Reference)
  ]}
]
variables [
 Skeletal_Mesh {type=SkeletalMeshComponent}
 thigh_rot_l {type=number}
 thigh_rot_r {type=number}
]
